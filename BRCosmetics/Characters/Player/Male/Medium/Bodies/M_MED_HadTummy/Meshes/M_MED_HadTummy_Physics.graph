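# Physics asset for the M_MED_HadTummy male player body cosmetic, driving hanging front and back cloth through bone bodies.
# Six four-link capsule chains (center, left, right, front and back) hang from base bodies joined freely to spine_01.
# Chain links use tight twist and swing limits, allowing wider motion on one swing axis, more in front than back.
# Spine, glute, thigh and foot capsules add body volumes, with feet joined freely to thighs.
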PhysicsAsset /BRCosmetics/Characters/Player/Male/Medium/Bodies/M_MED_HadTummy/Meshes/M_MED_HadTummy_Physics
bodies {
 base_backCloth_01 [{shapeType=Capsule, length=1, radius=1, shapeName=base_frontCloth_01_capsule, rotation=<rot 0 0 90>}]
 base_backCloth_01_l [{shapeType=Capsule, length=1, radius=1, shapeName=base_frontCloth_01_capsule, rotation=<rot 0 0 90>}]
 base_backCloth_01_r [{shapeType=Capsule, length=1, radius=1, shapeName=base_frontCloth_01_capsule, rotation=<rot 0 0 90>}]
 base_frontCloth_01 [{shapeType=Capsule, length=1, radius=1, shapeName=base_frontCloth_01_capsule, rotation=<rot 0 0 90>}]
 base_frontCloth_01_l [{shapeType=Capsule, length=1, radius=1, shapeName=base_frontCloth_01_capsule, rotation=<rot 0 0 90>}]
 base_frontCloth_01_r [{shapeType=Capsule, length=1, radius=1, shapeName=base_frontCloth_01_capsule, rotation=<rot 0 0 90>}]
 deform_glute_base_l [{shapeType=Capsule, length=13, radius=7, shapeName=deform_glute_base_l_capsule, center=<7 0 0>}]
 deform_glute_base_r [{shapeType=Capsule, length=10, radius=5, shapeName=deform_glute_base_l_capsule, center=<-7 0 0>}]
 dyn_backCloth_01 [{shapeType=Capsule, length=5, radius=5, shapeName=dyn_backCloth_01_capsule, center=<2.7667 0.275 -0.46>, rotation=<rot 0 1.2762 -89.57>}]
 dyn_backCloth_01_l [{shapeType=Capsule, length=3, radius=3, shapeName=dyn_frontCloth_02_l_capsule, center=<4.2907 -0.8609 0.8637>, rotation=<rot 0 0 90>}]
 dyn_backCloth_01_r [{shapeType=Capsule, length=3, radius=3, shapeName=dyn_frontCloth_02_l_capsule, center=<-4.4408 -2.9245 -0.5015>, rotation=<rot 0 0 90>}]
 dyn_backCloth_02 [{
   shapeType=Capsule
   length=5
   radius=5.5
   shapeName=dyn_backCloth_02_capsule
   center=<3.5104 0.5202 -0.8655>
   rotation=<rot 0 -1.3424 -88.3388>
  }]
 dyn_backCloth_02_l [{shapeType=Capsule, length=3, radius=4, shapeName=dyn_frontCloth_02_l_capsule, center=<4.2907 -0.8609 0.8637>, rotation=<rot 0 0 90>}]
 dyn_backCloth_02_r [{shapeType=Capsule, length=3, radius=3.5, shapeName=dyn_frontCloth_02_l_capsule, center=<-4.4408 -2.9245 -0.5015>, rotation=<rot 0 0 90>}]
 dyn_backCloth_03 [{
   shapeType=Capsule
   length=5
   radius=5.5
   shapeName=dyn_backCloth_03_capsule
   center=<3.4529 0.1763 -0.8423>
   rotation=<rot 0 -1.9985 -89.9878>
  }]
 dyn_backCloth_03_l [{shapeType=Capsule, length=5, radius=5.5, shapeName=dyn_frontCloth_02_l_capsule, center=<4.2907 -0.8609 0.8637>, rotation=<rot 0 0 90>}]
 dyn_backCloth_03_r [{shapeType=Capsule, length=5, radius=5.5, shapeName=dyn_frontCloth_02_l_capsule, center=<-4.4408 -2.9245 -0.5015>, rotation=<rot 0 0 90>}]
 dyn_backCloth_04 [{
   shapeType=Capsule
   length=5
   radius=5.5
   shapeName=dyn_backCloth_04_capsule
   center=<3.6595 0.1291 -1.0519>
   rotation=<rot 0 179.2576 -89.2059>
  }]
 dyn_backCloth_04_l [{shapeType=Capsule, length=5, radius=5.5, shapeName=dyn_frontCloth_02_l_capsule, center=<4.2907 -0.8609 0.8637>, rotation=<rot 0 0 90>}]
 dyn_backCloth_04_r [{shapeType=Capsule, length=5, radius=5.5, shapeName=dyn_frontCloth_02_l_capsule, center=<-4.4408 -2.9245 -0.5015>, rotation=<rot 0 0 90>}]
 dyn_frontCloth_01 [{shapeType=Capsule, length=20, radius=5, shapeName=dyn_frontCloth_02_l_capsule, center=<-1.396 -2.9245 1.0452>, rotation=<rot 0 0 90>}]
 dyn_frontCloth_01_l [{shapeType=Capsule, length=13, radius=5, shapeName=dyn_frontCloth_02_l_capsule, center=<0 0 5.3989>, rotation=<rot 0.0001 0.0001 98.5255>}]
 dyn_frontCloth_01_r [{shapeType=Capsule, length=13, radius=5, shapeName=dyn_frontCloth_02_l_capsule, center=<0 0 -4.3803>, rotation=<rot 0.0001 0.0001 98.5255>}]
 dyn_frontCloth_02 [{shapeType=Capsule, length=20, radius=5, shapeName=dyn_frontCloth_02_l_capsule, center=<-1.3654 -2.9245 3.0724>, rotation=<rot 0 0 90>}]
 dyn_frontCloth_02_l [{
   shapeType=Capsule
   length=13
   radius=5
   shapeName=dyn_frontCloth_02_l_capsule
   center=<2.3981 -0.0572 5.0171>
   rotation=<rot 0.0001 0.0001 98.5255>
  }]
 dyn_frontCloth_02_r [{
   shapeType=Capsule
   length=13
   radius=5
   shapeName=dyn_frontCloth_02_l_capsule
   center=<-2.4028 -0.0576 -5.3725>
   rotation=<rot 0.0002 0.0002 78.5255>
  }]
 dyn_frontCloth_03 [{shapeType=Capsule, length=22, radius=5, shapeName=dyn_frontCloth_02_l_capsule, center=<0.2669 -2.9245 3.0724>, rotation=<rot 0 0 90>}]
 dyn_frontCloth_03_l [{
   shapeType=Capsule
   length=13
   radius=6
   shapeName=dyn_frontCloth_02_l_capsule
   center=<4.1146 -0.0826 4.4773>
   rotation=<rot 0.0001 0.0001 98.5255>
  }]
 dyn_frontCloth_03_r [{
   shapeType=Capsule
   length=13
   radius=5.5
   shapeName=dyn_frontCloth_02_l_capsule
   center=<-3.4404 -0.0818 -4.4668>
   rotation=<rot 0.0002 0.0001 78.5255>
  }]
 dyn_frontCloth_04 [{shapeType=Capsule, length=27, radius=5, shapeName=dyn_frontCloth_02_l_capsule, center=<3.6628 -2.9245 3.0724>, rotation=<rot 0 0 90>}]
 dyn_frontCloth_04_l [{
   shapeType=Capsule
   length=13
   radius=6
   shapeName=dyn_frontCloth_02_l_capsule
   center=<4.9893 0 3.9741>
   rotation=<rot 0.0001 0.0001 98.5255>
  }]
 dyn_frontCloth_04_r [{
   shapeType=Capsule
   length=13
   radius=5.5
   shapeName=dyn_frontCloth_02_l_capsule
   center=<-5.1786 -0.1162 -3.4199>
   rotation=<rot 0.0003 0.0002 78.5255>
  }]
 foot_l [{shapeType=Capsule, length=14.7126, radius=10, shapeName=foot_r_capsule}]
 foot_r [{shapeType=Capsule, length=14.7126, radius=10, shapeName=foot_r_capsule}]
 spine_01 [{shapeType=Capsule, length=10, radius=9, shapeName=spine_01_capsule, center=<-0.6971 -0.7045 0.0242>, rotation=<rot 0 0 5>}]
 thigh_l [{shapeType=Capsule, length=20, radius=9.6, shapeName=thigh_l_capsule, center=<-10 -0.5 0>, rotation=<rot 90 0 0>}]
 thigh_r [
  {shapeType=Capsule, length=22, radius=9.5, shapeName=thigh_r_capsule, center=<10 0.6 0>, rotation=<rot 90 180 0>}
  {shapeType=Capsule, length=20, radius=9, shapeName=thigh_r_capsule, center=<10 0 -10>, rotation=<rot 90 180 0>}
 ]
}
constraints [
 {
  bone1Name=base_frontCloth_01
  bone2Name=spine_01
  swing1LimitDegrees=45
  swing1Motion=Free
  swing2LimitDegrees=45
  swing2Motion=Free
  twistLimitDegrees=45
  twistMotion=Free
 }
 {bone1Name=dyn_frontCloth_01, bone2Name=base_frontCloth_01, swing1LimitDegrees=1, swing2LimitDegrees=30, twistLimitDegrees=5}
 {bone1Name=dyn_frontCloth_02, bone2Name=dyn_frontCloth_01, swing1LimitDegrees=1, swing2LimitDegrees=30, twistLimitDegrees=5}
 {bone1Name=dyn_frontCloth_03, bone2Name=dyn_frontCloth_02, swing1LimitDegrees=1, swing2LimitDegrees=30, twistLimitDegrees=5}
 {bone1Name=dyn_frontCloth_04, bone2Name=dyn_frontCloth_03, swing1LimitDegrees=1, swing2LimitDegrees=30, twistLimitDegrees=5}
 {
  bone1Name=base_frontCloth_01_l
  bone2Name=spine_01
  swing1LimitDegrees=45
  swing1Motion=Free
  swing2LimitDegrees=45
  swing2Motion=Free
  twistLimitDegrees=45
  twistMotion=Free
 }
 {
  bone1Name=base_frontCloth_01_r
  bone2Name=spine_01
  swing1LimitDegrees=45
  swing1Motion=Free
  swing2LimitDegrees=45
  swing2Motion=Free
  twistLimitDegrees=45
  twistMotion=Free
 }
 {bone1Name=dyn_frontCloth_01_l, bone2Name=base_frontCloth_01_l, swing1LimitDegrees=1, swing2LimitDegrees=30, twistLimitDegrees=5}
 {bone1Name=dyn_frontCloth_01_r, bone2Name=base_frontCloth_01_r, swing1LimitDegrees=1, swing2LimitDegrees=30, twistLimitDegrees=5}
 {bone1Name=dyn_frontCloth_02_l, bone2Name=dyn_frontCloth_01_l, swing1LimitDegrees=1, swing2LimitDegrees=30, twistLimitDegrees=5}
 {bone1Name=dyn_frontCloth_02_r, bone2Name=dyn_frontCloth_01_r, swing1LimitDegrees=1, swing2LimitDegrees=30, twistLimitDegrees=5}
 {bone1Name=dyn_frontCloth_03_l, bone2Name=dyn_frontCloth_02_l, swing1LimitDegrees=1, swing2LimitDegrees=30, twistLimitDegrees=5}
 {bone1Name=dyn_frontCloth_03_r, bone2Name=dyn_frontCloth_02_r, swing1LimitDegrees=1, swing2LimitDegrees=30, twistLimitDegrees=5}
 {bone1Name=dyn_frontCloth_04_l, bone2Name=dyn_frontCloth_03_l, swing1LimitDegrees=1, swing2LimitDegrees=30, twistLimitDegrees=5}
 {bone1Name=dyn_frontCloth_04_r, bone2Name=dyn_frontCloth_03_r, swing1LimitDegrees=1, swing2LimitDegrees=30, twistLimitDegrees=5}
 {
  bone1Name=base_backCloth_01
  bone2Name=spine_01
  swing1LimitDegrees=45
  swing1Motion=Free
  swing2LimitDegrees=45
  swing2Motion=Free
  twistLimitDegrees=45
  twistMotion=Free
 }
 {
  bone1Name=base_backCloth_01_l
  bone2Name=spine_01
  swing1LimitDegrees=45
  swing1Motion=Free
  swing2LimitDegrees=45
  swing2Motion=Free
  twistLimitDegrees=45
  twistMotion=Free
 }
 {
  bone1Name=base_backCloth_01_r
  bone2Name=spine_01
  swing1LimitDegrees=45
  swing1Motion=Free
  swing2LimitDegrees=45
  swing2Motion=Free
  twistLimitDegrees=45
  twistMotion=Free
 }
 {bone1Name=dyn_backCloth_01, bone2Name=base_backCloth_01, swing1LimitDegrees=1, swing2LimitDegrees=20, twistLimitDegrees=5}
 {bone1Name=dyn_backCloth_02, bone2Name=dyn_backCloth_01, swing1LimitDegrees=1, swing2LimitDegrees=20, twistLimitDegrees=5}
 {bone1Name=dyn_backCloth_03, bone2Name=dyn_backCloth_02, swing1LimitDegrees=1, swing2LimitDegrees=20, twistLimitDegrees=5}
 {bone1Name=dyn_backCloth_04, bone2Name=dyn_backCloth_03, swing1LimitDegrees=1, swing2LimitDegrees=20, twistLimitDegrees=5}
 {bone1Name=dyn_backCloth_01_l, bone2Name=base_backCloth_01_l, swing1LimitDegrees=1, swing2LimitDegrees=20, twistLimitDegrees=5}
 {bone1Name=dyn_backCloth_02_l, bone2Name=dyn_backCloth_01_l, swing1LimitDegrees=1, swing2LimitDegrees=20, twistLimitDegrees=5}
 {bone1Name=dyn_backCloth_03_l, bone2Name=dyn_backCloth_02_l, swing1LimitDegrees=1, swing2LimitDegrees=20, twistLimitDegrees=5}
 {bone1Name=dyn_backCloth_04_l, bone2Name=dyn_backCloth_03_l, swing1LimitDegrees=1, swing2LimitDegrees=20, twistLimitDegrees=5}
 {bone1Name=dyn_backCloth_01_r, bone2Name=base_backCloth_01_r, swing1LimitDegrees=1, swing2LimitDegrees=20, twistLimitDegrees=5}
 {bone1Name=dyn_backCloth_02_r, bone2Name=dyn_backCloth_01_r, swing1LimitDegrees=1, swing2LimitDegrees=20, twistLimitDegrees=5}
 {bone1Name=dyn_backCloth_03_r, bone2Name=dyn_backCloth_02_r, swing1LimitDegrees=1, swing2LimitDegrees=20, twistLimitDegrees=5}
 {bone1Name=dyn_backCloth_04_r, bone2Name=dyn_backCloth_03_r, swing1LimitDegrees=1, swing2LimitDegrees=20, twistLimitDegrees=5}
 {
  bone1Name=foot_r
  bone2Name=thigh_r
  swing1LimitDegrees=45
  swing1Motion=Free
  swing2LimitDegrees=45
  swing2Motion=Free
  twistLimitDegrees=45
  twistMotion=Free
 }
 {
  bone1Name=foot_l
  bone2Name=thigh_l
  swing1LimitDegrees=45
  swing1Motion=Free
  swing2LimitDegrees=45
  swing2Motion=Free
  twistLimitDegrees=45
  twistMotion=Free
 }
]
